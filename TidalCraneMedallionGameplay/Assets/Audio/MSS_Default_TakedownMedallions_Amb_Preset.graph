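# MetaSound source for the ambient loop of default takedown medallions, played as a quiet mono sound.
# On play, a single mono Wave Player plays the medallion ambience wave, with the asset and loop flag exposed as inputs, and feeds the mono output.
# It uses item ambient-loop attenuation and the loot collect sound class.
# A concurrency override caps simultaneous voices and scales volume by distance.
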MetaSoundSource /TidalCraneMedallionGameplay/Assets/Audio/MSS_Default_TakedownMedallions_Amb_Preset
inputs [
 Loop {type=Bool}
 UE.Source.OnPlay {type=Trigger}
 Wave Asset {type=WaveAsset}
]
interfaces [UE.OutputFormat.Mono:1.0, UE.Source:1.1]
outputs ["UE.OutputFormat.Mono.Audio:0" {type=Audio}]
pages [Default {
  input_defaults {
   Loop="<MetaSoundBuilderResult.SUCCEEDED: 0>"
   UE.Source.OnPlay=null
   Wave Asset=&/Default_TakedownMedallions_Amb
  }
  nodes [WavePlayer_0 = UE.Wave Player:Mono(Loop=^Loop, Play=^UE.Source.OnPlay, Wave Asset=^Wave Asset) {also {version [1,1]}}]
  links [{dst="@UE.OutputFormat.Mono.Audio:0", src=WavePlayer_0.Out Mono}]
 }]
properties {
 assetImportData=&self:AssetImportData
 attenuationSettings=/Game/Sounds/Attenuation/Fort_Item_Ambient_Loop_Mono_Default_Att
 bLooping=true
 bOverrideConcurrency=true
 duration=10000
 maxDistance=1250
 sampleRate=48000
 soundClassObject=&Fort_Loot_Collect
 volume=0.2
 concurrencyOverrides {maxCount=2, voiceStealReleaseTime=0.3, volumeScale=0.4, volumeScaleAttackTime=0.1, volumeScaleMode=Distance}
}
