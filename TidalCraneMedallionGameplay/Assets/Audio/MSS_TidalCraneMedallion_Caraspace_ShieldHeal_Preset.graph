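# Stereo MetaSound source for the Tidal Crane Medallion Caraspace shield-heal effect, set to looping with item-interact attenuation.
# StartRepeat drives a repeating trigger, with clamped Rate as its period in seconds, that plays random burst sounds from RepeatOS.
# A delayed fade after OnPlay halves the burst level before the stereo mixer output.
# Loop Sounds, envelope, volume-range and scan inputs are declared but unconnected in this graph.
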MetaSoundSource /TidalCraneMedallionGameplay/Assets/Audio/MSS_TidalCraneMedallion_Caraspace_ShieldHeal_Preset
inputs [
 Attack Time {type=Time}
 Loop {type=Bool}
 Loop Sounds {type=WaveAsset:Array}
 MaxVolume {type=Float}
 MinVolume {type=Float}
 Rate {type=Float}
 Ratio {type=Float}
 Release Time {type=Time}
 RepeatOS {type=WaveAsset:Array}
 Scan.Duration {type=Float}
 StartRepeat {type=Trigger}
 UE.Source.OnPlay {type=Trigger}
]
interfaces [UE.OutputFormat.Stereo:1.0, UE.Source:1.1]
outputs [
 "UE.OutputFormat.Stereo.Audio:0" {type=Audio}
 "UE.OutputFormat.Stereo.Audio:1" {type=Audio}
 WaveDuration {type=Time}
]
pages [Default {
  input_defaults {
   Attack Time=0.2
   Loop="<MetaSoundBuilderResult.SUCCEEDED: 0>"
   MaxVolume=0
   MinVolume=0
   Rate=2
   Ratio=1.5
   Release Time=0.204
   Scan.Duration=1.64
   StartRepeat=null
   UE.Source.OnPlay=null
   Loop Sounds [&/TidalCraneMedallion_Caraspace_ShieldHeal_Loop_01]
   RepeatOS [
    &/TidalCraneMedallion_Caraspace_ShieldHeal_Bursts_01
    &/TidalCraneMedallion_Caraspace_ShieldHeal_Bursts_02
    &/TidalCraneMedallion_Caraspace_ShieldHeal_Bursts_03
    &/TidalCraneMedallion_Caraspace_ShieldHeal_Bursts_04
   ]
  }
  nodes [
   TriggerDelay_0 = UE.Trigger Delay(In=^UE.Source.OnPlay) {also {defaults {Delay Time=5}}}
   n1 = /FortMetasound/Patches/MSP_FadeToValue(Fade=@TriggerDelay_0.Out) {also {defaults {FadeTime=3, Initial Value=1, TargetValue=0.5}}}
   n2 = Clamp.Clamp:Float(In=^Rate) {also {defaults {Max=20, Min=2}}}
   n3 = &MSS_PlayRandomSound_Stereo(Sounds=^RepeatOS, UE.Source.OnPlay=UE.TriggerRepeat(Start=^StartRepeat, Period=UE.ConversionFloatToTime(In=@n2.Value).Out).RepeatOut) {also {defaults {Volume=0.25}}}
   AudioMixerAudioMixerStereo2_0 = "AudioMixer.Audio Mixer (Stereo, 2)"(
    In 1 R=UE.Multiply:Audio by Float(AdditionalOperands=@n1.Out Value, PrimaryOperand=@n3."UE.OutputFormat.Stereo.Audio:1").Out
    In 1 L=UE.Multiply:Audio by Float(AdditionalOperands=@n1.Out Value, PrimaryOperand=@n3."UE.OutputFormat.Stereo.Audio:0").Out
   )
   InitVariable.Time()
   InitVariable.Time()
   InitVariable.Trigger()
  ]
  links [
   {dst="@UE.OutputFormat.Stereo.Audio:0", src=AudioMixerAudioMixerStereo2_0.Out L}
   {dst="@UE.OutputFormat.Stereo.Audio:1", src=AudioMixerAudioMixerStereo2_0.Out R}
  ]
 }]
properties {
 assetImportData=&self:AssetImportData
 attenuationSettings=/Game/Sounds/Attenuation/Fort_Item_Interact_Att
 bLooping=true
 duration=10000
 maxDistance=1850
 numChannels=2
 outputFormat=Stereo
 sampleRate=48000
 soundClassObject=&Fort_Gadgets_Items
 volume=0.7
}
