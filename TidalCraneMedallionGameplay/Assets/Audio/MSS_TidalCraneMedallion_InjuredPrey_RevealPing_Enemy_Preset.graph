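# One-shot 5.1 MetaSound for the enemy reveal ping of the Tidal Crane Medallion's Injured Prey effect.
# On play, a character affiliation check gates a random pick from the EnemyPing array, played across all six channels.
# The same trigger starts a mono IP sound after a short delay, mixed into front left and right with a variable-driven gain.
# The source finishes when that delayed sound ends, and uses in-game ping attenuation with limited concurrency.
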MetaSoundSource /TidalCraneMedallionGameplay/Assets/Audio/MSS_TidalCraneMedallion_InjuredPrey_RevealPing_Enemy_Preset
inputs [
 Character.Affiliation {type=Enum:FortAudioAffiliation}
 Character.Health {type=Float}
 Character.UniqueID {type=Int32}
 EnemyPing {type=WaveAsset:Array}
 IP Sound {type=WaveAsset}
 UE.Source.OnPlay {type=Trigger}
]
interfaces [Fort.Character:1.0, UE.OutputFormat.5dot1:1.0, UE.Source.OneShot:1.0, UE.Source:1.1]
outputs [
 Out Side Left {type=Audio}
 Out Side Right {type=Audio}
 "UE.OutputFormat.5dot1.Audio:0" {type=Audio}
 "UE.OutputFormat.5dot1.Audio:1" {type=Audio}
 "UE.OutputFormat.5dot1.Audio:2" {type=Audio}
 "UE.OutputFormat.5dot1.Audio:3" {type=Audio}
 "UE.OutputFormat.5dot1.Audio:4" {type=Audio}
 "UE.OutputFormat.5dot1.Audio:5" {type=Audio}
 UE.Source.OneShot.OnFinished {type=Trigger}
]
pages [Default {
  input_defaults {
   Character.Affiliation=0
   Character.Health=0
   Character.UniqueID=0
   IP Sound=&/IP_TidalCraneMedallion_InjuredPrey_ChiMa_01
   UE.Source.OnPlay=null
   EnemyPing [&/TidalCraneMedallion_InjuredPrey_RevealPing_Enemy_01]
  }
  nodes [
   n1 = VariableAccessor.Float(Variable=InitVariable.Float().Variable)
   n2 = &TriggerCompareAffiliation(A=^Character.Affiliation, Compare=^UE.Source.OnPlay)
   ArrayRandomGet_0 = Array.Random Get:WaveAsset:Array(In Array=^EnemyPing, Next=@n2.True)
   WavePlayer_0 = UE.Wave Player:5dot1(Play=@ArrayRandomGet_0.On Next, Wave Asset=@ArrayRandomGet_0.Value) {also {version [1,1]}}
   TriggerDelay_0 = UE.Trigger Delay(In=@n2.True) {also {defaults {Delay Time=0.5}}}
   WavePlayer_1 = UE.Wave Player:Mono(Wave Asset=^IP Sound, Play=@TriggerDelay_0.Out) {also {version [1,1]}}
   AudioMixerAudioMixerMono2_0 = "AudioMixer.Audio Mixer (Mono, 2)"(Gain 1=@n1.Value, In 0=@WavePlayer_0.Out Front Right, In 1=@WavePlayer_1.Out Mono)
   AudioMixerAudioMixerMono2_1 = "AudioMixer.Audio Mixer (Mono, 2)"(Gain 1=@n1.Value, In 0=@WavePlayer_0.Out Front Left, In 1=@WavePlayer_1.Out Mono)
  ]
  links [
   {dst="@UE.OutputFormat.5dot1.Audio:1", src=AudioMixerAudioMixerMono2_0.Out}
   {dst="@UE.OutputFormat.5dot1.Audio:0", src=AudioMixerAudioMixerMono2_1.Out}
   {dst="@UE.OutputFormat.5dot1.Audio:2", src=WavePlayer_0.Out Front Center}
   {dst="@UE.OutputFormat.5dot1.Audio:3", src=WavePlayer_0.Out Low Frequency}
   {dst="@UE.OutputFormat.5dot1.Audio:4", src=WavePlayer_0.Out Side Left}
   {dst="@UE.OutputFormat.5dot1.Audio:5", src=WavePlayer_0.Out Side Right}
   {dst="@UE.Source.OneShot.OnFinished", src=WavePlayer_1.On Finished}
  ]
 }]
properties {
 assetImportData=&self:AssetImportData
 attenuationSettings=/Game/Sounds/Attenuation/Fort_UI_Ping_Ingame_Att
 bOverrideConcurrency=true
 numChannels=6
 outputFormat=FiveDotOne
 sampleRate=48000
 soundClassObject=&Fort_Gadgets_Items
 volume=0.3
 concurrencyOverrides {maxCount=3}
}
