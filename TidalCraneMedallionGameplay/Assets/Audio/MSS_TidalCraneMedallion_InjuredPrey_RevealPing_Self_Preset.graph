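# One-shot mono MetaSound for the Tidal Crane Medallion's Injured Prey reveal ping, the "Self" variant.
# On play, a Character.Affiliation compare gates everything: its True branch plays a random SelfPing wave and, after a short delay, the IP Sound layer.
# Both players are mixed with the delayed layer at reduced gain, and only that layer's end fires OnFinished.
# It uses in-game ping attenuation and a small concurrency cap.
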MetaSoundSource /TidalCraneMedallionGameplay/Assets/Audio/MSS_TidalCraneMedallion_InjuredPrey_RevealPing_Self_Preset
inputs [
 Character.Affiliation {type=Enum:FortAudioAffiliation}
 Character.Health {type=Float}
 Character.UniqueID {type=Int32}
 IP Sound {type=WaveAsset}
 SelfPing {type=WaveAsset:Array}
 UE.Source.OnPlay {type=Trigger}
]
interfaces [Fort.Character:1.0, UE.OutputFormat.Mono:1.0, UE.Source.OneShot:1.0, UE.Source:1.1]
outputs [
 "UE.OutputFormat.Mono.Audio:0" {type=Audio}
 UE.Source.OneShot.OnFinished {type=Trigger}
]
pages [Default {
  input_defaults {
   Character.Affiliation=0
   Character.Health=0
   Character.UniqueID=0
   IP Sound=&/IP_TidalCraneMedallion_InjuredPrey_ChiMa_01
   UE.Source.OnPlay=null
   SelfPing [&/TidalCraneMedallion_InjuredPrey_RevealPing_Self_01]
  }
  nodes [
   n1 = &TriggerCompareAffiliation(A=^Character.Affiliation, Compare=^UE.Source.OnPlay)
   ArrayRandomGet_0 = Array.Random Get:WaveAsset:Array(In Array=^SelfPing, Next=@n1.True)
   WavePlayer_0 = UE.Wave Player:Mono(Play=@ArrayRandomGet_0.On Next, Wave Asset=@ArrayRandomGet_0.Value) {also {version [1,1]}}
   TriggerDelay_0 = UE.Trigger Delay(In=@n1.True) {also {defaults {Delay Time=0.3}}}
   WavePlayer_1 = UE.Wave Player:Mono(Wave Asset=^IP Sound, Play=@TriggerDelay_0.Out) {also {version [1,1]}}
   AudioMixerAudioMixerMono2_0 = "AudioMixer.Audio Mixer (Mono, 2)"(In 0=@WavePlayer_0.Out Mono, In 1=@WavePlayer_1.Out Mono) {also {defaults {Gain 1=0.5}}}
  ]
  links [
   {dst="@UE.OutputFormat.Mono.Audio:0", src=AudioMixerAudioMixerMono2_0.Out}
   {dst="@UE.Source.OneShot.OnFinished", src=WavePlayer_1.On Finished}
  ]
 }]
properties {
 assetImportData=&self:AssetImportData
 attenuationSettings=/Game/Sounds/Attenuation/Fort_UI_Ping_Ingame_Att
 bOverrideConcurrency=true
 sampleRate=48000
 soundClassObject=&Fort_Gadgets_Items
 volume=0.3
 concurrencyOverrides {maxCount=3}
}
