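# One-shot stereo MetaSound Source for the Tidal Crane Medallion respawn transform sound, using skill-use attenuation and the gadgets/items sound class.
# Acts as a preset of MSS_PlayRandomSound_Stereo, forwarding all inputs to that shared graph and returning its stereo audio, finished trigger and wave duration.
# The Sounds array holds a single wave, and the pitch and volume ranges are flat, so playback does not vary as configured.
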MetaSoundSource /TidalCraneMedallionGameplay/Assets/Audio/MSS_TidalCraneMedallion_Respawn_Transform_Preset
inputs [
 MaxPitch {type=Float}
 MaxVolume {type=Float}
 MinPitch {type=Float}
 MinVolume {type=Float}
 Pitch {type=Float}
 Seed {type=Int32}
 Shared State Behavior {type=Enum:SharedStateBehaviorType}
 Sounds {type=WaveAsset:Array}
 Start Time {type=Time}
 UE.Source.OnPlay {type=Trigger}
 Volume {type=Float}
]
interfaces [UE.OutputFormat.Stereo:1.0, UE.Source.OneShot:1.0, UE.Source:1.1]
outputs [
 "UE.OutputFormat.Stereo.Audio:0" {type=Audio}
 "UE.OutputFormat.Stereo.Audio:1" {type=Audio}
 UE.Source.OneShot.OnFinished {type=Trigger}
 WaveDuration {type=Time}
]
pages [Default {
  input_defaults {
   MaxPitch=1
   MaxVolume=1
   MinPitch=1
   MinVolume=1
   Pitch=1
   Seed=-1
   Shared State Behavior=1
   Start Time=0
   UE.Source.OnPlay=null
   Volume=1
   Sounds [&/TidalCraneMedallion_Respawn_Transform_01]
  }
  nodes [MSS_PlayRandomSound_Stereo_0 = &MSS_PlayRandomSound_Stereo(
    MaxPitch=^MaxPitch
    MaxVolume=^MaxVolume
    MinPitch=^MinPitch
    MinVolume=^MinVolume
    Pitch=^Pitch
    Seed=^Seed
    Shared State Behavior=^Shared State Behavior
    Sounds=^Sounds
    Start Time=^Start Time
    UE.Source.OnPlay=^UE.Source.OnPlay
    Volume=^Volume
   )]
  links [
   {dst="@UE.OutputFormat.Stereo.Audio:0", src=MSS_PlayRandomSound_Stereo_0.UE.OutputFormat.Stereo.Audio:0}
   {dst="@UE.OutputFormat.Stereo.Audio:1", src=MSS_PlayRandomSound_Stereo_0.UE.OutputFormat.Stereo.Audio:1}
   {dst="@UE.Source.OneShot.OnFinished", src=MSS_PlayRandomSound_Stereo_0.UE.Source.OneShot.OnFinished}
   {dst="@WaveDuration", src=MSS_PlayRandomSound_Stereo_0.WaveDuration}
  ]
 }]
properties {
 assetImportData=&self:AssetImportData
 attenuationSettings=/Game/Sounds/Attenuation/Fort_SkillUse_Att
 maxDistance=4000
 numChannels=2
 outputFormat=Stereo
 sampleRate=48000
 soundClassObject=&Fort_Gadgets_Items
}
