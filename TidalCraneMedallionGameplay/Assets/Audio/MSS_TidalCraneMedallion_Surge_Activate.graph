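# One-shot stereo MetaSound that plays the activation sound for the Tidal Crane Medallion's Surge.
# It wraps the shared MSS_PlayRandomSound_Stereo graph, forwarding a wave array plus base and min/max pitch and volume for slight random variation.
# The array holds a single wave by default, and the graph also outputs wave duration.
# It uses gadget-activate attenuation, the gadgets sound class and a concurrency override limiting it to one instance.
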MetaSoundSource /TidalCraneMedallionGameplay/Assets/Audio/MSS_TidalCraneMedallion_Surge_Activate
inputs [
 MaxPitch {type=Float}
 MaxVolume {type=Float}
 MinPitch {type=Float}
 MinVolume {type=Float}
 Pitch {type=Float}
 Sounds {type=WaveAsset:Array}
 UE.Source.OnPlay {type=Trigger}
 Volume {type=Float}
]
interfaces [UE.OutputFormat.Stereo:1.0, UE.Source.OneShot:1.0, UE.Source:1.1]
outputs [
 "UE.OutputFormat.Stereo.Audio:0" {type=Audio}
 "UE.OutputFormat.Stereo.Audio:1" {type=Audio}
 UE.Source.OneShot.OnFinished {type=Trigger}
 WaveDuration {type=Time}
]
pages [Default {
  input_defaults {
   MaxPitch=1.05
   MaxVolume=1.05
   MinPitch=0.95
   MinVolume=0.95
   Pitch=1
   UE.Source.OnPlay=null
   Volume=1
   Sounds [&/TidalCraneMedallion_Surge_Activate_01]
  }
  nodes [MSS_PlayRandomSound_Stereo_0 = &MSS_PlayRandomSound_Stereo(
    MaxPitch=^MaxPitch
    MaxVolume=^MaxVolume
    MinPitch=^MinPitch
    MinVolume=^MinVolume
    Pitch=^Pitch
    Sounds=^Sounds
    UE.Source.OnPlay=^UE.Source.OnPlay
    Volume=^Volume
   )]
  links [
   {dst="@UE.OutputFormat.Stereo.Audio:0", src=MSS_PlayRandomSound_Stereo_0.UE.OutputFormat.Stereo.Audio:0}
   {dst="@UE.OutputFormat.Stereo.Audio:1", src=MSS_PlayRandomSound_Stereo_0.UE.OutputFormat.Stereo.Audio:1}
   {dst="@UE.Source.OneShot.OnFinished", src=MSS_PlayRandomSound_Stereo_0.UE.Source.OneShot.OnFinished}
   {dst="@WaveDuration", src=MSS_PlayRandomSound_Stereo_0.WaveDuration}
  ]
 }]
properties {
 assetImportData=&self:AssetImportData
 attenuationSettings=&Fort_GadgetActivate_Att
 bOverrideConcurrency=true
 maxDistance=4000
 numChannels=2
 outputFormat=Stereo
 sampleRate=48000
 soundClassObject=&Fort_Gadgets_Items
 volume=2
 concurrencyOverrides {maxCount=1}
}
