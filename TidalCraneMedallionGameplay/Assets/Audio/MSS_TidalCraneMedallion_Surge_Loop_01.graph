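# One-shot stereo MetaSound for the Tidal Crane Medallion surge loop sound, using gadget attenuation and sound class, limited to one concurrent instance.
# On play, a Wave Player starts the surge wave and triggers an ADSR envelope's attack.
# The Duration of Loop float input sets a Trigger Delay that starts the envelope's release.
# The envelope multiplies both channels for the fade-out, and its completion fires OnFinished.
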MetaSoundSource /TidalCraneMedallionGameplay/Assets/Audio/MSS_TidalCraneMedallion_Surge_Loop_01
inputs [
 Duration of Loop {type=Float}
 UE.Source.OnPlay {type=Trigger}
]
interfaces [UE.OutputFormat.Stereo:1.0, UE.Source.OneShot:1.0, UE.Source:1.1]
outputs [
 "UE.OutputFormat.Stereo.Audio:0" {type=Audio}
 "UE.OutputFormat.Stereo.Audio:1" {type=Audio}
 UE.Source.OneShot.OnFinished {type=Trigger}
]
pages [Default {
  input_defaults {Duration of Loop=3, UE.Source.OnPlay=null}
  nodes [
   InitVariable.Time()
   InitVariable.Time()
   InitVariable.Trigger()
   WavePlayer_0 = UE.Wave Player:Stereo(Play=^UE.Source.OnPlay) {also {
     defaults {Loop="<MetaSoundBuilderResult.SUCCEEDED: 0>", Wave Asset=&/TidalCraneMedallion_Surge_Loop_01}
     version [1,1]
    }}
   TriggerDelay_0 = UE.Trigger Delay(Delay Time=UE.ConversionFloatToTime(In=^Duration of Loop).Out, In=@WavePlayer_0.On Play)
   ADSREnvelopeADSREnvelope_0 = ADSR Envelope.ADSR Envelope:Float(Trigger Release=@TriggerDelay_0.Out, Trigger Attack=@WavePlayer_0.On Play) {also {defaults {Release Curve=0.8, Release Time=3, Sustain Level=1}}}
   Multiply_0 = UE.Multiply:Audio by Float(AdditionalOperands=@ADSREnvelopeADSREnvelope_0.Out Envelope, PrimaryOperand=@WavePlayer_0.Out Left)
   Multiply_1 = UE.Multiply:Audio by Float(AdditionalOperands=@ADSREnvelopeADSREnvelope_0.Out Envelope, PrimaryOperand=@WavePlayer_0.Out Right)
  ]
  links [
   {dst="@UE.Source.OneShot.OnFinished", src=ADSREnvelopeADSREnvelope_0.On Done}
   {dst="@UE.OutputFormat.Stereo.Audio:0", src=Multiply_0.Out}
   {dst="@UE.OutputFormat.Stereo.Audio:1", src=Multiply_1.Out}
  ]
 }]
properties {
 assetImportData=&self:AssetImportData
 attenuationSettings=&Fort_GadgetActivate_Att
 bOverrideConcurrency=true
 maxDistance=4000
 numChannels=2
 outputFormat=Stereo
 sampleRate=48000
 soundClassObject=&Fort_Gadgets_Items
 volume=0.3
 concurrencyOverrides {maxCount=1}
}
