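# One-shot stereo MetaSound source for picking up the Tidal Crane Injured Prey takedown medallion.
# On play, it layers a music stinger and a randomly pitched default pickup sound, then adds a pitch-randomized IP activate sound after a short trigger delay.
# Stinger and activate layers pass through stereo delays before a gain-weighted three-input mixer, whose left and right outputs are wired crosswise.
# OnFinished follows the stinger alone, and concurrency is overridden to one instance.
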MetaSoundSource /TidalCraneMedallionGameplay/Assets/Audio/MSS_TidalCrane_TakedownMedallions_Pickup_InjuredPrey_Preset
inputs [
 DefaultPickup {type=WaveAsset}
 IP Activate {type=WaveAsset}
 Music Stinger {type=WaveAsset}
 UE.Source.OnPlay {type=Trigger}
]
interfaces [UE.OutputFormat.Stereo:1.0, UE.Source.OneShot:1.0, UE.Source:1.1]
outputs [
 "UE.OutputFormat.Stereo.Audio:0" {type=Audio}
 "UE.OutputFormat.Stereo.Audio:1" {type=Audio}
 UE.Source.OneShot.OnFinished {type=Trigger}
]
pages [Default {
  input_defaults {
   DefaultPickup=&/Default_TakedownMedallions_Pickup_SFX
   IP Activate=&/IP_TidalCraneMedallion_InjuredyPrey_Activate_01
   Music Stinger=&/MX_S37_Medallion_GnatGala_STNG
   UE.Source.OnPlay=null
  }
  nodes [
   TriggerDelay_0 = UE.Trigger Delay(In=^UE.Source.OnPlay) {also {defaults {Delay Time=0}}}
   WavePlayer_0 = UE.Wave Player:Stereo(Wave Asset=^Music Stinger, Play=@TriggerDelay_0.Out) {also {version [1,1]}}
   StereoDelay_0 = UE.Stereo Delay:Audio(In Left=@WavePlayer_0.Out Left, In Right=@WavePlayer_0.Out Right) {also {defaults {Delay Time=0.22, Max Delay Time=0.22}}}
   TriggerDelay_1 = UE.Trigger Delay(In=^UE.Source.OnPlay) {also {defaults {Delay Time=1.25}}}
   n1 = UE.RandomFloat(Next=@TriggerDelay_1.Out) {also {defaults {Max=0.5, Min=-0.5}}}
   WavePlayer_2 = UE.Wave Player:Stereo(Wave Asset=^IP Activate, Play=@n1.On Next, Pitch Shift=@n1.Value) {also {version [1,1]}}
   StereoDelay_1 = UE.Stereo Delay:Audio(In Left=@WavePlayer_2.Out Left, In Right=@WavePlayer_2.Out Right) {also {defaults {Delay Mode=1, Delay Time=2, Max Delay Time=0.22}}}
   n2 = UE.RandomFloat(Next=^UE.Source.OnPlay) {also {defaults {Max=0.5, Min=-0.5}}}
   WavePlayer_1 = UE.Wave Player:Stereo(Wave Asset=^DefaultPickup, Play=@n2.On Next, Pitch Shift=@n2.Value) {also {version [1,1]}}
   AudioMixerAudioMixerStereo3_0 = "AudioMixer.Audio Mixer (Stereo, 3)"(
    In 0 L=@StereoDelay_0.Out Left
    In 0 R=@StereoDelay_0.Out Right
    In 2 L=@StereoDelay_1.Out Left
    In 2 R=@StereoDelay_1.Out Right
    In 1 L=@WavePlayer_1.Out Left
    In 1 R=@WavePlayer_1.Out Right
   ) {also {defaults {Gain 0=1.2, Gain 2=0.5}}}
  ]
  links [
   {dst="@UE.OutputFormat.Stereo.Audio:1", src=AudioMixerAudioMixerStereo3_0.Out L}
   {dst="@UE.OutputFormat.Stereo.Audio:0", src=AudioMixerAudioMixerStereo3_0.Out R}
   {dst="@UE.Source.OneShot.OnFinished", src=WavePlayer_0.On Finished}
  ]
 }]
properties {
 assetImportData=&self:AssetImportData
 bOverrideConcurrency=true
 numChannels=2
 outputFormat=Stereo
 sampleRate=48000
 soundClassObject=&Fort_Loot_Collect
 volume=0.75
 concurrencyOverrides {maxCount=1}
}
