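# One-shot stereo MetaSound for the TidalCrane takedown medallion pickup, layering a pickup effect with a ToyFox music stinger.
# On play, a random-sound subgraph and a stinger Wave Player fire together; the stinger runs through a short Stereo Delay before both are mixed.
# OnFinished follows the undelayed stinger player, while WaveDuration reports the effect's length.
# The mixer's left and right outputs are cross-wired, and concurrency allows one instance.
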MetaSoundSource /TidalCraneMedallionGameplay/Assets/Audio/MSS_TidalCrane_TakedownMedallions_Pickup_Scanner_Preset
inputs [
 MaxPitch {type=Float}
 MaxVolume {type=Float}
 MinPitch {type=Float}
 MinVolume {type=Float}
 Music Stinger {type=WaveAsset}
 Pitch {type=Float}
 Seed {type=Int32}
 Shared State Behavior {type=Enum:SharedStateBehaviorType}
 Sounds {type=WaveAsset:Array}
 Start Time {type=Time}
 UE.Source.OnPlay {type=Trigger}
 Volume {type=Float}
]
interfaces [UE.OutputFormat.Stereo:1.0, UE.Source.OneShot:1.0, UE.Source:1.1]
outputs [
 "UE.OutputFormat.Stereo.Audio:0" {type=Audio}
 "UE.OutputFormat.Stereo.Audio:1" {type=Audio}
 UE.Source.OneShot.OnFinished {type=Trigger}
 WaveDuration {type=Time}
]
pages [Default {
  input_defaults {
   MaxPitch=1
   MaxVolume=1
   MinPitch=1
   MinVolume=1
   Music Stinger=&/MX_S37_Medallion_ToyFox_STNG
   Pitch=1
   Seed=-1
   Shared State Behavior=1
   Start Time=0
   UE.Source.OnPlay=null
   Volume=1
   Sounds [&/Default_TakedownMedallions_Pickup_SFX]
  }
  nodes [
   MSS_PlayRandomSound_Stereo_0 = &MSS_PlayRandomSound_Stereo(
    MaxPitch=^MaxPitch
    MaxVolume=^MaxVolume
    MinPitch=^MinPitch
    MinVolume=^MinVolume
    Pitch=^Pitch
    Seed=^Seed
    Shared State Behavior=^Shared State Behavior
    Sounds=^Sounds
    Start Time=^Start Time
    UE.Source.OnPlay=^UE.Source.OnPlay
    Volume=^Volume
   )
   WavePlayer_0 = UE.Wave Player:Stereo(Wave Asset=^Music Stinger, Play=^UE.Source.OnPlay) {also {version [1,1]}}
   StereoDelay_0 = UE.Stereo Delay:Audio(In Left=@WavePlayer_0.Out Left, In Right=@WavePlayer_0.Out Right) {also {defaults {Delay Time=0.22, Max Delay Time=0.22}}}
   AudioMixerAudioMixerStereo2_0 = "AudioMixer.Audio Mixer (Stereo, 2)"(
    In 1 L=@MSS_PlayRandomSound_Stereo_0."UE.OutputFormat.Stereo.Audio:0"
    In 1 R=@MSS_PlayRandomSound_Stereo_0."UE.OutputFormat.Stereo.Audio:1"
    In 0 L=@StereoDelay_0.Out Left
    In 0 R=@StereoDelay_0.Out Right
   )
  ]
  links [
   {dst="@UE.OutputFormat.Stereo.Audio:1", src=AudioMixerAudioMixerStereo2_0.Out L}
   {dst="@UE.OutputFormat.Stereo.Audio:0", src=AudioMixerAudioMixerStereo2_0.Out R}
   {dst="@WaveDuration", src=MSS_PlayRandomSound_Stereo_0.WaveDuration}
   {dst="@UE.Source.OneShot.OnFinished", src=WavePlayer_0.On Finished}
  ]
 }]
properties {
 assetImportData=&self:AssetImportData
 bOverrideConcurrency=true
 numChannels=2
 outputFormat=Stereo
 sampleRate=48000
 soundClassObject=&Fort_Loot_Collect
 volume=0.75
 concurrencyOverrides {maxCount=1}
}
